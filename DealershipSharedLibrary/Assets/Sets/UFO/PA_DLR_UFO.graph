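# Physics asset for the UFO set piece in the dealership shared library, giving its skeletal mesh a simple collision shape.
# It holds a single body on the Rotate bone, a large capsule whose length nearly equals its radius, offset upward from the bone.
# There are no constraints, so it defines no joints between bodies.
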PhysicsAsset /DealershipSharedLibrary/Assets/Sets/UFO/PA_DLR_UFO
bodies {Rotate [{shapeType=Capsule, length=312.2826, radius=312.2808, center=<0.0608 0.187 51.4192>, rotation=<rot 0.1931 -107.9998 -90>}]}
constraints []
